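# Physics asset for the wasteland truck body mesh in the vehicle cosmetics set, giving it a simplified collision shape.
# It holds a single body on the frame bone, approximated by one large capsule rotated onto its side.
# There are no constraints, so the asset defines no jointed or separately simulated parts.
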PhysicsAsset /VehicleCosmetics/Bodies/WastelandTruck/Meshes/SM_WastelandTruck_PhysicsAsset
bodies {frame [{shapeType=Capsule, length=224.8322, radius=151.2233, center=<-8.2109 -5.857 1.163>, rotation=<rot 0 -85.2237 -89.6537>}]}
constraints []
